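# Debug calibrometer widget for the Sparks calibration map, derived from WBP_CalibrometerWidget with no widgets of its own.
# On initialization it calls the parent, then sets the MusicClock variable from its GetMusicClock function.
# GetMusicClock finds the owning player's Sparks Music Playspace, matching child types, and returns that playspace's music clock, or none if no valid playspace exists.
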WidgetBlueprint /SparksMaps/Calibration/WBP_CalibrometerDebug
graphs [
 EventGraph {nodes [
   Event OnInitialized()
   > Parent: OnInitialized() {is=title K2Node_CallParentFunction}
   > n1 = GetMusicClock!(execute, self) -> (then, MusicClock:Music Clock Object Reference)
   > Set MusicClock(MusicClock:Music Clock Object Reference=@n1) -> (Output_Get:Music Clock Object Reference)
   Event PreConstruct()
   Event Construct()
  ]}
 GetMusicClock {nodes [
   GetMusicClock() {is=title K2Node_FunctionEntry}
   n1 = GetSparksMusicPlayspaceOfClass!(
    PlayspaceType:Sparks Music Playspace Class Reference=/Script/SparksMusicPlayspaceRuntime.SparksMusicPlayspace
    Actor:Actor Object Reference=GetOwningPlayer()
    bMatchChildTypes:bool=true
   ) -> (ReturnValue:Sparks Music Playspace Object Reference)
   > n2 = Is Valid(InputObject=@n1)
   Return Node!(execute=@n2.Is Not Valid, MusicClock:Music Clock Object Reference)
   > Return Node(MusicClock=Get MusicClock!(self:Sparks Music Playspace Object Reference=@n1) -> (MusicClock:Music Clock Object Reference))
  ]}
]
variables []
widgets {
 info {
  inheritedWidgetCount=0
  parentClass=/FMCalibration/Widgets/WBP_CalibrometerWidget.WBP_CalibrometerWidget_C
  rootWidgetClass=null
  widgetCount=0
 }
 widgets []
}
